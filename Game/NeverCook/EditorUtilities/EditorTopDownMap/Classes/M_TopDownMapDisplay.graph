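# UI-domain material for the editor top-down map display, combining a map texture with an overlay texture.
# MapTexture is partially desaturated by Desat and scaled by DarkenMap, then lerped toward OverlayTexture RGB using the overlay alpha times BlendOverlay.
# BlendOverlay is left at its default, so the overlay stays hidden until that parameter is raised.
# A stray divide of the map's red channel by "fads" does not feed the blend.
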Material /Game/NeverCook/EditorUtilities/EditorTopDownMap/Classes/M_TopDownMapDisplay
graph {
 properties {materialDomain=MD_UI, pixelDepthOffsetMode=PDOM_Legacy, usedShadingModels=Default Lit}
 nodes [
  n1 = TextureSampleParameter2D() {parameterName=MapTexture, samplerType=SAMPLERTYPE_LinearColor, texture=/Game/Global/Textures/Default/Blanks/T_BlackLinearColor}
  Divide(A=@n1.R, B=ScalarParameter() {defaultValue=2, parameterName=fads})
  n2 = ScalarParameter() {defaultValue=0.5, parameterName=Desat}
  n3 = TextureSampleParameter2D() {parameterName=OverlayTexture, samplerType=SAMPLERTYPE_LinearColor, texture=&T_BlackLinearColor}
  LinearInterpolate(
   A=Multiply(A=ScalarParameter() {defaultValue=0.5, parameterName=DarkenMap}, B=Desaturation(None=@n1.RGB, Fraction=@n2))
   B=@n3.RGB
   Alpha=Multiply(A=@n3.A, B=ScalarParameter() {parameterName=BlendOverlay})
  )
 ]
}
