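# Editor utility options overlay for the top-down map widget, with a gear button that toggles a blurred settings panel.
# On construct it finds its parent EUW_TopDownMapWidget and binds the panel's DetailsViews to it, exposing bounds, capture actor and blend properties.
# The useActorBounds setting controls visibility of the actor tag field.
# Recapture and the camera button call the parent's RefreshMap, and the open panel consumes mouse clicks.
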EditorUtilityWidgetBlueprint /Game/NeverCook/EditorUtilities/EditorTopDownMap/EUW_TopDownMapOptions
graphs [
 EventGraph {nodes [
   "On Clicked (btnX)"!() -> (OutputDelegate, then, widget:User Widget Object Reference, event:Pointer Event Structure) {is=title K2Node_ComponentBoundEvent}
   > SetVisibility(self=Get BackgroundBlur_157() -> (BackgroundBlur_157:Background Blur Object Reference), InVisibility=Collapsed)
   "On Clicked (btnGear)"() {is=title K2Node_ComponentBoundEvent}
   n1 = Get BackgroundBlur_157()
   > SetVisibility(self=@n1, InVisibility=Select(Option 0=Visible, Option 1=Collapsed, Index=IsVisible(self=@n1)))
   "On Property Changed (detailsUseActorBounds)"!() -> (OutputDelegate, then, PropertyName:Name=None) {is=title K2Node_ComponentBoundEvent}
   > validateActorBoundsDisplay()
   n2 = "On Clicked (btnCapture)"!() -> (OutputDelegate, then, widget:User Widget Object Reference, event:Pointer Event Structure) {is=title K2Node_ComponentBoundEvent}
   n3 = Get ParentWidget() -> (ParentWidget:EUW Top Down Map Widget Object Reference)
   Is Valid(exec=[@n4.then, @n2.then], InputObject=@n3)
   > RefreshMap(self:EUW Top Down Map Widget Object Reference=@n3)
   "On Clicked (btnCaptureTest)"() {is=title K2Node_ComponentBoundEvent}
   > n4 = PrintString(InString=Append(A="][][][][][][][][][][][ ", B=GetDisplayName(Object=@n3)))
   Event Construct()
   > n5 = FindParentWidgetOfType!(
    execute
    StartingWidget:Widget Object Reference=Self-Reference()
    Type:Widget Class Reference=&/EUW_TopDownMapWidget.EUW_TopDownMapWidget_C
   ) -> (then, ReturnValue:EUW Top Down Map Widget Object Reference)
   > Is Valid(InputObject=@n5)
   > Set ParentWidget(ParentWidget:EUW Top Down Map Widget Object Reference=@n5) -> (Output_Get:EUW Top Down Map Widget Object Reference)
   > validateActorBoundsDisplay()
   > n6 = For Each Loop(Array:Array of Widget Object References=GetAllChildren(self=Get VerticalBox_0() -> (VerticalBox_0:Vertical Box Object Reference))) -> (Array Element:Widget Object Reference)
   > n7 = Cast To DetailsView!(execute, Object=@n6) -> (then, CastFailed, AsDetails View:Details View Object Reference) {is=title K2Node_DynamicCast}
   > SetObject(self=@n7, NewObject=Get ParentWidget())
  ]}
 OnMouseButtonDown {nodes [
   OnMouseButtonDown!() -> (then, MyGeometry:Geometry Structure, MouseEvent:"Pointer Event Structure (by ref)") {is=title K2Node_FunctionEntry}
   n1 = Get BackgroundBlur_157()
   > Return Node!(
    execute
    ReturnValue:Event Reply Structure=Select(Option 0:Event Reply Structure=Unhandled(), Option 1:Event Reply Structure=Handled(), Index=IsVisible(self=@n1)) -> (ReturnValue:Event Reply Structure)
   )
  ]}
 validateActorBoundsDisplay {nodes [
   validateActorBoundsDisplay() {is=title K2Node_FunctionEntry}
   n1 = Get ParentWidget()
   > Is Valid(InputObject=@n1)
   n2 = Get useActorBounds!(self:EUW Top Down Map Widget Object Reference=@n1) -> (useActorBounds:bool)
   > SetVisibility(
    self=Get detailsActorName() -> (detailsActorName:Details View Object Reference)
    InVisibility=Select(Option 0:ESlateVisibility Enum=Collapsed, Option 1:ESlateVisibility Enum=Visible, Index=@n2) -> (ReturnValue:ESlateVisibility Enum)
   )
   SetVisibility(
    self=Get detailsBoundsSet() -> (detailsBoundsSet:Details View Object Reference)
    InVisibility=Select(Option 0=Visible, Option 1=Collapsed, Index=@n2)
   )
  ]}
]
variables [ParentWidget {type=EUW_TopDownMapWidget_C, default=null}]
widgets {
 info {rootWidgetClass=&Overlay, widgetCount=12}
 widgets [
  Overlay_0 {widgetClassPath=&Overlay, widget=Overlay_0}
  btnGear {
   widgetClassPath=&EditorUtilityButton
   parent=Overlay_0
   widget=btnGear
   properties {widgetStyle {
     clickedSlateSound {resourceObject=null}
     disabled {
      bIsDynamicallyLoaded=false
      drawAs=Image
      imageType=NoImage
      mirroring=NoMirror
      resourceName=None
      resourceObject=/Game/Athena/UI/InGame/Creative/PropertyEditor/Icon_Picker/Textures_V2/64x64/T_UI_IconLibrary_Gear_64
      tiling=NoTile
      imageSize=<32 32>
      margin=<ltrb 0 0 0 0>
      outlineSettings {
       bUseBrushTransparency=false
       roundingType=FixedRadius
       width=1
       color {colorUseRule=UseColor_ColorTable, specifiedColor=<rgba 0 0 0 0>}
       cornerRadii=<4 4 4 4>
      }
      tintColor {colorUseRule=UseColor_Specified, specifiedColor=<rgba 0.1823 0.1823 0.1823 1>}
      uVRegion {bIsValid=false, max=<0 0>, min=<0 0>}
     }
     disabledForeground {colorUseRule=UseColor_ColorTable, specifiedColor=<rgba 0 0 0 0>}
     hovered {
      bIsDynamicallyLoaded=false
      drawAs=Image
      imageType=NoImage
      mirroring=NoMirror
      resourceName=None
      resourceObject=&T_UI_IconLibrary_Gear_64
      tiling=NoTile
      imageSize=<32 32>
      margin=<ltrb 0 0 0 0>
      outlineSettings {
       bUseBrushTransparency=false
       roundingType=FixedRadius
       width=1
       color {colorUseRule=UseColor_ColorTable, specifiedColor=<rgba 0 0 0 0>}
       cornerRadii=<4 4 4 4>
      }
      tintColor {colorUseRule=UseColor_Specified, specifiedColor=<rgba 0.7396 0.7396 0.7396 1>}
      uVRegion {bIsValid=false, max=<0 0>, min=<0 0>}
     }
     hoveredForeground {colorUseRule=UseColor_ColorTable, specifiedColor=<rgba 0 0 0 0>}
     hoveredSlateSound {resourceObject=null}
     normal {
      bIsDynamicallyLoaded=false
      drawAs=Image
      imageType=NoImage
      mirroring=NoMirror
      resourceName=None
      resourceObject=&T_UI_IconLibrary_Gear_64
      tiling=NoTile
      imageSize=<32 32>
      margin=<ltrb 0 0 0 0>
      outlineSettings {
       bUseBrushTransparency=false
       roundingType=FixedRadius
       width=1
       color {colorUseRule=UseColor_ColorTable, specifiedColor=<rgba 0 0 0 0>}
       cornerRadii=<4 4 4 4>
      }
      tintColor {colorUseRule=UseColor_Specified, specifiedColor=<rgba 0.1823 0.1823 0.1823 0.664>}
      uVRegion {bIsValid=false, max=<0 0>, min=<0 0>}
     }
     normalForeground {colorUseRule=UseColor_ColorTable, specifiedColor=<rgba 0 0 0 0>}
     normalPadding=<ltrb 12 1.5 12 1.5>
     pressed {
      bIsDynamicallyLoaded=false
      drawAs=Image
      imageType=NoImage
      mirroring=NoMirror
      resourceName=None
      resourceObject=&T_UI_IconLibrary_Gear_64
      tiling=NoTile
      imageSize=<32 32>
      margin=<ltrb 0 0 0 0>
      outlineSettings {
       bUseBrushTransparency=false
       roundingType=FixedRadius
       width=1
       color {colorUseRule=UseColor_ColorTable, specifiedColor=<rgba 0 0 0 0>}
       cornerRadii=<4 4 4 4>
      }
      tintColor {colorUseRule=UseColor_Specified, specifiedColor=<rgba 0.125 0.125 0.125 1>}
      uVRegion {bIsValid=false, max=<0 0>, min=<0 0>}
     }
     pressedForeground {colorUseRule=UseColor_ColorTable, specifiedColor=<rgba 0 0 0 0>}
     pressedPadding=<ltrb 12 2.5 12 0.5>
     pressedSlateSound {resourceObject=null}
    }}
   slot {class=OverlaySlot, properties {horizontalAlignment=HAlign_Right}}
  }
  btnCaptureTest {
   widgetClassPath=&EditorUtilityButton
   parent=Overlay_0
   widget=btnCaptureTest
   properties {widgetStyle {
     clickedSlateSound {resourceObject=null}
     disabled {
      bIsDynamicallyLoaded=false
      drawAs=Image
      imageType=NoImage
      mirroring=NoMirror
      resourceName=None
      resourceObject=/Game/Athena/UI/InGame/Creative/PropertyEditor/Icon_Picker/Textures_V2/64x64/T_UI_IconLibrary_Camera_64
      tiling=NoTile
      imageSize=<32 32>
      margin=<ltrb 0 0 0 0>
      outlineSettings {
       bUseBrushTransparency=false
       roundingType=FixedRadius
       width=1
       color {colorUseRule=UseColor_ColorTable, specifiedColor=<rgba 0 0 0 0>}
       cornerRadii=<4 4 4 4>
      }
      tintColor {colorUseRule=UseColor_Specified, specifiedColor=<rgba 0.1823 0.1823 0.1823 1>}
      uVRegion {bIsValid=false, max=<0 0>, min=<0 0>}
     }
     disabledForeground {colorUseRule=UseColor_ColorTable, specifiedColor=<rgba 0 0 0 0>}
     hovered {
      bIsDynamicallyLoaded=false
      drawAs=Image
      imageType=NoImage
      mirroring=NoMirror
      resourceName=None
      resourceObject=&T_UI_IconLibrary_Camera_64
      tiling=NoTile
      imageSize=<32 32>
      margin=<ltrb 0 0 0 0>
      outlineSettings {
       bUseBrushTransparency=false
       roundingType=FixedRadius
       width=1
       color {colorUseRule=UseColor_ColorTable, specifiedColor=<rgba 0 0 0 0>}
       cornerRadii=<4 4 4 4>
      }
      tintColor {colorUseRule=UseColor_Specified, specifiedColor=<rgba 0.7396 0.7396 0.7396 1>}
      uVRegion {bIsValid=false, max=<0 0>, min=<0 0>}
     }
     hoveredForeground {colorUseRule=UseColor_ColorTable, specifiedColor=<rgba 0 0 0 0>}
     hoveredSlateSound {resourceObject=null}
     normal {
      bIsDynamicallyLoaded=false
      drawAs=Image
      imageType=NoImage
      mirroring=NoMirror
      resourceName=None
      resourceObject=&T_UI_IconLibrary_Camera_64
      tiling=NoTile
      imageSize=<32 32>
      margin=<ltrb 0 0 0 0>
      outlineSettings {
       bUseBrushTransparency=false
       roundingType=FixedRadius
       width=1
       color {colorUseRule=UseColor_ColorTable, specifiedColor=<rgba 0 0 0 0>}
       cornerRadii=<4 4 4 4>
      }
      tintColor {colorUseRule=UseColor_Specified, specifiedColor=<rgba 0.1823 0.1823 0.1823 0.664>}
      uVRegion {bIsValid=false, max=<0 0>, min=<0 0>}
     }
     normalForeground {colorUseRule=UseColor_ColorTable, specifiedColor=<rgba 0 0 0 0>}
     normalPadding=<ltrb 12 1.5 12 1.5>
     pressed {
      bIsDynamicallyLoaded=false
      drawAs=Image
      imageType=NoImage
      mirroring=NoMirror
      resourceName=None
      resourceObject=&T_UI_IconLibrary_Camera_64
      tiling=NoTile
      imageSize=<32 32>
      margin=<ltrb 0 0 0 0>
      outlineSettings {
       bUseBrushTransparency=false
       roundingType=FixedRadius
       width=1
       color {colorUseRule=UseColor_ColorTable, specifiedColor=<rgba 0 0 0 0>}
       cornerRadii=<4 4 4 4>
      }
      tintColor {colorUseRule=UseColor_Specified, specifiedColor=<rgba 0.125 0.125 0.125 1>}
      uVRegion {bIsValid=false, max=<0 0>, min=<0 0>}
     }
     pressedForeground {colorUseRule=UseColor_ColorTable, specifiedColor=<rgba 0 0 0 0>}
     pressedPadding=<ltrb 12 2.5 12 0.5>
     pressedSlateSound {resourceObject=null}
    }}
   slot {class=OverlaySlot, properties {horizontalAlignment=HAlign_Right, padding=<ltrb 0 35 0 0>}}
  }
  BackgroundBlur_157 {
   widgetClassPath=&BackgroundBlur
   bIsVariable=true
   parent=Overlay_0
   widget=BackgroundBlur_157
   properties {blurStrength=0.6, visibility=Collapsed, padding=<ltrb 20 20 20 20>}
   slot {class=OverlaySlot, properties {horizontalAlignment=HAlign_Fill, verticalAlignment=VAlign_Fill}}
  }
  Border_0 {
   widgetClassPath=&Border
   bIsVariable=true
   parent=BackgroundBlur_157
   widget=Border_0
   properties {brushColor=<rgba 0.0176 0.0176 0.0176 1>, padding=<ltrb 5 5 5 5>}
   slot {class=BackgroundBlurSlot, properties {padding=<ltrb 20 20 20 20>}}
  }
  VerticalBox_0 {
   widgetClassPath=&VerticalBox
   bIsVariable=true
   parent=Border_0
   widget=VerticalBox_0
   slot {class=BorderSlot, properties {padding=<ltrb 5 5 5 5>}}
  }
  btnX {
   widgetClassPath=&TA_Generic01_Button
   parent=VerticalBox_0
   widget=btnX
   properties {
    buttonColor_Default=<rgba 0.6667 0.6667 0.6667 1>
    iconBrush {
     bIsDynamicallyLoaded=false
     drawAs=Image
     imageType=NoImage
     mirroring=NoMirror
     resourceName=None
     resourceObject=/Game/UI/Icons/Icon_Seach_Close
     tiling=NoTile
     imageSize=<16 16>
     margin=<ltrb 0 0 0 0>
     outlineSettings {
      bUseBrushTransparency=false
      roundingType=HalfHeightRadius
      width=0
      color {colorUseRule=UseColor_Specified, specifiedColor=<rgba 0 0 0 0>}
      cornerRadii=<0 0 0 1>
     }
     tintColor {colorUseRule=UseColor_Specified, specifiedColor=<rgba 0.0417 0.0417 0.0417 1>}
     uVRegion {bIsValid=false, max=<0 0>, min=<0 0>}
    }
   }
   slot {class=VerticalBoxSlot, properties {horizontalAlignment=HAlign_Right}}
  }
  detailsUseActorBounds {
   widgetClassPath=&DetailsView
   parent=VerticalBox_0
   widget=detailsUseActorBounds
   properties {propertiesToShow [useActorBounds]}
   slot {class=VerticalBoxSlot}
  }
  detailsActorName {
   widgetClassPath=&DetailsView
   parent=VerticalBox_0
   widget=detailsActorName
   properties {propertiesToShow [ActorTagForBounds]}
   slot {class=VerticalBoxSlot}
  }
  detailsBoundsSet {
   widgetClassPath=&DetailsView
   parent=VerticalBox_0
   widget=detailsBoundsSet
   properties {propertiesToShow [BackupBoundsMin, BackupBoundsMax]}
   slot {class=VerticalBoxSlot}
  }
  detailsGeneral {
   widgetClassPath=&DetailsView
   parent=VerticalBox_0
   widget=detailsGeneral
   properties {propertiesToShow [CaptureActorBP, OverlayBlend]}
   slot {class=VerticalBoxSlot}
  }
  btnCapture {
   widgetClassPath=&TA_Generic01_Button
   parent=VerticalBox_0
   widget=btnCapture
   properties {label=Recapture}
   slot {class=VerticalBoxSlot}
  }
 ]
}
