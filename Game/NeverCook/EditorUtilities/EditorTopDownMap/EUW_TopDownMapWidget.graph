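# Editor utility widget showing a top-down level map that tracks and moves the viewport camera.
# Bounds from a tagged actor, or backup min/max, size an orthographic capture by a temporary spawned actor into a render target, displayed through a dynamic material with optional overlay texture.
# Tick updates a camera marker, clamped and restyled when off-map.
# Clicking teleports the camera; dragging also sets yaw, with OnPaint drawing the drag line.
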EditorUtilityWidgetBlueprint /Game/NeverCook/EditorUtilities/EditorTopDownMap/EUW_TopDownMapWidget
graphs [
 CancelTeleportDraw {nodes [
   CancelTeleportDraw() {is=title K2Node_FunctionEntry}
   > Set MouseDownLoc(MouseDownLoc:Vector 2D Structure=(X=-1,Y=-1)) -> (Output_Get:Vector 2D Structure)
  ]}
 CaptureMapRT {nodes [
   CaptureMapRT() {is=title K2Node_FunctionEntry}
   n1 = Break Box 2D!(Box2D:"Box 2D Structure (by ref)"=Get CurrentBounds() -> (CurrentBounds:Box 2D Structure)) -> (Min:Vector 2D Structure, Max:Vector 2D Structure, bIsValid:bool=~) {is=title K2Node_BreakStruct}
   > Branch(Condition=@n1.bIsValid)
   n2 = BreakVector2D(InVec=@n1)
   n3 = BreakVector2D(InVec=@n1.Max)
   n4 = float + float(A=@n2.Y, B=@n3.Y)
   n5 = float + float(A=@n2, B=@n3)
   > n6 = SpawnActorFromObject!(
    execute
    ObjectToUse:Object Reference=Get CaptureActorBP() -> (CaptureActorBP:Object Reference)
    Location:Vector=MakeVector(X=float / float(A=@n5, B=2), Y=float / float(A=@n4, B=2), Z=15000)
    Rotation:Rotator
    bTransient:bool=true
   ) -> (then, ReturnValue:Actor Object Reference)
   > n7 = Set l_spawnedActor(l_spawnedActor:Actor Object Reference=@n6) -> (Output_Get:Actor Object Reference)
   > n8 = Set l_captureComp(l_captureComp:Scene Capture Component 2D Object Reference=GetComponentByClass!(self:Actor Object Reference=@n7, ComponentClass:Actor Component Class Reference=/Script/Engine.SceneCaptureComponent2D) -> (ReturnValue:Scene Capture Component 2D Object Reference)) -> (Output_Get:Scene Capture Component 2D Object Reference)
   > Is Valid(InputObject=@n8)
   > n9 = Sequence!(execute) -> (then_0, then_1, then_2:Exec, then_3:Exec, then_4:Exec)
   n10 = Break Box 2D(Box2D=Get CurrentBounds()) -> (bIsValid=~)
   > Set OrthoWidth!(
    execute
    OrthoWidth:float=GetAbsMax2D!(A:Vector 2D Structure=vector2d - vector2d(A:Vector 2D Structure=@n10.Max, B:Vector 2D Structure=@n10) -> (ReturnValue:Vector 2D Structure) {is=title K2Node_PromotableOperator}) -> (ReturnValue:double)
    self:Scene Capture Component 2D Object Reference=Get l_captureComp() -> (l_captureComp:Scene Capture Component 2D Object Reference)
   ) -> (then, Output_Get:float)
   n11 = Is Valid(exec=@n9.then_1, InputObject=Get CurrentRT() -> (CurrentRT:Texture Render Target 2D Object Reference))
   n12 = CreateRenderTarget2D!(
    execute=@n11.Is Not Valid
    WorldContextObject:Object Reference=Self-Reference()
    Width:int=512
    Height:int=512
    Format:ETextureRenderTargetFormat Enum=RTF_RGBA8
    ClearColor:Linear Color Structure=(R=0,G=0,B=0,A=1)
    bAutoGenerateMipMaps:bool=false
    bSupportUAVs:bool=false
   ) -> (then, ReturnValue:Texture Render Target 2D Object Reference)
   > n13 = Set CurrentRT(CurrentRT:Texture Render Target 2D Object Reference=@n12) -> (Output_Get:Texture Render Target 2D Object Reference)
   Set TextureTarget!(
    execute=[@n11.Is Valid, @n13.then]
    TextureTarget:Texture Render Target 2D Object Reference=Get CurrentRT()
    self:Scene Capture Component 2D Object Reference=Get l_captureComp()
   ) -> (then, Output_Get:Texture Render Target 2D Object Reference)
   CaptureScene(execute=@n9.then_2, self:Scene Capture Component 2D Object Reference=Get l_captureComp())
   DestroyActor!(
    execute=@n9.then_3
    self:Editor Actor Subsystem Object Reference=Get EditorActorSubsystem()
    ActorToDestroy:Actor Object Reference=Get l_spawnedActor() -> (l_spawnedActor:Actor Object Reference)
   ) -> (then, ReturnValue:bool)
   ValidateMapMaterial(execute=@n9.then_4)
  ]}
 DrawEditorCameraOnRT {nodes [
   n1 = DrawEditorCameraOnRT!() -> (then, force:bool=~) {is=title K2Node_FunctionEntry}
   > n2 = Get Editor Camera Info!(execute) -> (then, CameraLocation:Vector, ForwardDirection:Vector, FOV:float, ReturnValue:bool) {is=title K2Node_CallFunction}
   > Set l_camLoc(l_camLoc:Vector=@n2) -> (Output_Get:Vector)
   > Set l_camDir(l_camDir:Vector=@n2.ForwardDirection) -> (Output_Get:Vector)
   n3 = Get lastCamDirection() -> (lastCamDirection:Vector)
   n4 = Get l_camDir() -> (l_camDir:Vector)
   > n5 = Branch(Condition=OR Boolean(A="Not Equal Exactly (Vector)"!(A:Vector=@n3, B:Vector=@n4) -> (ReturnValue:bool), B=@n1))
   n6 = Sequence!(execute=[@n31.then, @n5.then]) -> (then_0, then_1, then_2)
   > Set lastCamLocation(lastCamLocation:Vector=Get l_camLoc() -> (l_camLoc:Vector)) -> (Output_Get:Vector)
   > Set lastCamDirection(lastCamDirection:Vector=Get l_camDir()) -> (Output_Get:Vector)
   n7 = vector * vector(A=Get lastCamDirection(), B:Vector="1,1,0") -> (ReturnValue=~)
   Set l_angle(
    execute=@n6.then_1
    l_angle:float=VectorToNormalizedAngleInDegrees!(V:Vector=Normalize!(A:Vector=@n7, Tolerance:float=0.0001) -> (ReturnValue:Vector) {is=title K2Node_CallFunction}) -> (ReturnValue:float)
   ) -> (Output_Get:float)
   n8 = Break Box 2D(Box2D=Get CurrentBounds()) -> (bIsValid=~)
   n9 = BreakVector2D(InVec=@n8)
   n10 = BreakVector(InVec=Get lastCamLocation() -> (lastCamLocation:Vector))
   n11 = BreakVector2D(InVec=@n8.Max)
   > Set l_camLoc2D(l_camLoc2D:Vector 2D Structure=MakeVector2D(
     X=NormalizeToRange!(Value:double=@n10, RangeMin:double=@n9, RangeMax:double=@n11) -> (ReturnValue:double)
     Y=NormalizeToRange(Value=@n10.Y, RangeMin=@n9.Y, RangeMax=@n11.Y)
    )) -> (Output_Get:Vector 2D Structure)
   n12 = GetLocalSize!(Geometry:"Geometry Structure (by ref)"=GetPaintSpaceGeometry!(self:Widget Object Reference=Get imgRT() -> (imgRT:Image Object Reference)) -> (ReturnValue:"Geometry Structure (by ref)")) -> (ReturnValue:Vector 2D Structure)
   n13 = BreakVector2D(InVec=vector2d * vector2d(A:Vector 2D Structure=@n12, B:Vector 2D Structure=Get l_camLoc2D() -> (l_camLoc2D:Vector 2D Structure)) -> (ReturnValue:Vector 2D Structure) {is=title K2Node_PromotableOperator})
   n14 = Set l_left(execute=@n6.then_2, l_left:double=@n13) -> (Output_Get:double)
   > n15 = Set l_right(l_right:double=@n13.Y) -> (Output_Get:double)
   n16 = BreakVector2D(InVec=@n12)
   n17 = GetDesiredSize!(self:Widget Object Reference=Get imgCam() -> (imgCam:Image Object Reference)) -> (ReturnValue_X:double, ReturnValue_Y:double)
   n18 = float - float(A=@n16.Y, B=@n17.ReturnValue_Y)
   n19 = float - float(A=@n16, B=@n17)
   > Set l_outsideWidget(l_outsideWidget:bool=OR Boolean!(A="float < float"(A=@n14), B="float < float"(A=@n15), C:bool="float > float"(A=@n14, B=@n19), D:bool="float > float"(A=@n15, B=@n18)) -> (ReturnValue)) -> (Output_Get:bool)
   > n20 = Cast To OverlaySlot!(execute, Object=Get Slot!(self:Widget Object Reference=Get imgCam()) -> (Slot:Panel Slot Object Reference)) -> (then, CastFailed, AsOverlay Slot:Overlay Slot Object Reference) {is=title K2Node_DynamicCast}
   n21 = Get l_right() -> (l_right:double)
   n22 = Get l_left() -> (l_left:double)
   > SetPadding!(
    execute
    self:Overlay Slot Object Reference=@n20
    InPadding:Margin Structure=Make Margin(Left="Clamp (Float)"(Value=@n22, Min=0, Max=@n19), Top="Clamp (Float)"(Value=@n21, Min=0, Max=@n18), Right=0, Bottom=0)
   ) -> (then)
   n23 = Get l_angle() -> (l_angle:float)
   n24 = Break Box 2D(Box2D=Get CurrentBounds()) -> (bIsValid=~)
   n25 = BreakVector2D(InVec=@n24)
   n26 = BreakVector2D(InVec=@n24.Max)
   n27 = MakeVector(X=float / float(A=float + float(A=@n25, B=@n26), B=2), Y=float / float(A=float + float(A=@n25.Y, B=@n26.Y), B=2))
   n28 = vector * vector(A=Get l_camLoc(), B="1,1,0") -> (ReturnValue=~)
   n29 = VectorToNormalizedAngleInDegrees(V=Normalize(A=vector - vector(A=@n27, B=@n28), Tolerance=0.0001))
   n30 = Get l_outsideWidget() -> (l_outsideWidget:bool)
   > SetRenderTransformAngle!(
    execute
    self:Widget Object Reference=Get imgCam()
    Angle:float=Select(Option 0:float=float + float(A=@n23, B=90), Option 1:float=float + float(A=@n29, B=-90), Index=@n30) -> (ReturnValue:float)
   ) -> (then)
   > SetRenderOpacity(self=Get imgCam(), InOpacity=Select(Option 0=1, Option 1=0.5, Index=@n30))
   > SetRenderScale!(
    execute
    self:Widget Object Reference=Get imgCam()
    Scale:Vector 2D Structure=To Vector2D!(InDouble:double=Select(Option 0=1, Option 1=0.5, Index=@n30)) -> (ReturnValue:Vector 2D Structure)
   ) -> (then)
   > SetBrushFromTexture!(
    execute
    self:Image Object Reference=Get imgCam()
    Texture:Texture 2D Object Reference=Select(
     Option 0:Texture 2D Object Reference=/Game/UI/Icons/Icon_CameraWithFrustum
     Option 1:Texture 2D Object Reference=/Game/UI/CodeReferenced/UI_Radar_AthenaPlayerArrow
     Index=@n30
    ) -> (ReturnValue:Texture 2D Object Reference)
    bMatchSize:bool=false
   ) -> (then)
   n31 = Branch(execute=@n5.else, Condition="Not Equal Exactly (Vector)"(A=Get lastCamLocation(), B=Get l_camLoc()))
  ]}
 EventGraph {nodes [
   RefreshMap() {is=title K2Node_CustomEvent}
   > TryGetBounds!(execute, self) -> (then, CurrentBounds:Box 2D Structure)
   > Delay!(execute, WorldContextObject:Object Reference, Duration=0.2) -> (then)
   > DrawEditorCameraOnRT!(execute, self, force:bool=true) -> (then) {is=title K2Node_CallFunction}
   > CaptureMapRT() {is=title K2Node_CallFunction}
   n1 = Event OnMouseEnter()
   > n2 = Branch(Condition=Is Mouse Button Down!(Input:"Pointer Event Structure (by ref)"=@n1.MouseEvent, MouseButton:Key Structure=LeftMouseButton) -> (ReturnValue:bool))
   CancelTeleportDraw(execute=@n2.else) {is=title K2Node_CallFunction}
   Event Construct()
   > RefreshMap() {is=title K2Node_CallFunction}
   Event Tick()
   > n3 = GetIfMapActive!(execute, self) -> (then, isMapActive:bool)
   > Branch(Condition=@n3)
   > DrawEditorCameraOnRT(force=false)
  ]}
 GetBoundsFromActorByTag {nodes [
   GetBoundsFromActorByTag() {is=title K2Node_FunctionEntry}
   > n1 = GetAllActorsWithTag!(execute, WorldContextObject:Object Reference, Tag:Name=Get ActorTagForBounds() -> (ActorTagForBounds:Name=None)) -> (then, OutActors:Array of Actor Object References)
   > Branch(Condition=Is Valid Index!(TargetArray:Array of Actor Object References=@n1, IndexToTest:int) -> (ReturnValue:bool))
   n2 = Get Actor Bounds(self="Get (a copy)"(Array:Array of Actor Object References=@n1) -> (Output:Actor Object Reference), bOnlyCollidingComponents=false)
   n3 = BreakVector(InVec=vector + vector(A=@n2, B=@n2.BoxExtent))
   n4 = BreakVector(InVec=vector - vector(A=@n2, B=@n2.BoxExtent))
   > Return Node!(
    execute
    ReturnValue:Box 2D Structure=MakeBox2D!(Min:Vector 2D Structure=MakeVector2D(X=@n4, Y=@n4.Y), Max:Vector 2D Structure=MakeVector2D(X=@n3, Y=@n3.Y)) -> (ReturnValue:Box 2D Structure)
   )
  ]}
 GetCameraPositionFrom2DLoc {nodes [
   n1 = GetCameraPositionFrom2DLoc!() -> (then, AbsoluteCoordinate:Vector 2D Structure) {is=title K2Node_FunctionEntry}
   n2 = GetCachedGeometry!(self:Widget Object Reference=Get imgRT()) -> (ReturnValue:"Geometry Structure (by ref)")
   n3 = Get imgRT()
   n4 = BreakVector2D(InVec=vector2d / vector2d(
     A:Vector 2D Structure="Absolute to Local (Coordinates)"!(Geometry:"Geometry Structure (by ref)"=@n2, AbsoluteCoordinate:Vector 2D Structure=@n1) -> (ReturnValue:Vector 2D Structure)
     B:Vector 2D Structure=GetDesiredSize!(self=@n3) -> (ReturnValue:Vector 2D Structure)
    ) -> (ReturnValue:Vector 2D Structure) {is=title K2Node_PromotableOperator})
   n5 = Break Box 2D(Box2D=Get CurrentBounds()) -> (bIsValid=~)
   n6 = BreakVector2D(InVec=@n5)
   n7 = BreakVector2D(InVec=@n5.Max)
   n8 = Get UnrealEditorSubsystem()
   n9 = GetLevelViewportCameraInfo!(self:Unreal Editor Subsystem Object Reference=@n8) -> (CameraLocation:Vector, CameraRotation:Rotator, ReturnValue:bool)
   n10 = MakeVector(X=Lerp(A=@n6, B=@n7, Alpha=@n4), Y=Lerp(A=@n6.Y, B=@n7.Y, Alpha=@n4.Y), Z=BreakVector(InVec=@n9).Z)
   > Return Node!(execute, location:Vector=@n10, rotation:Rotator=@n9.CameraRotation)
   SetLevelViewportCameraInfo!(execute, self:Unreal Editor Subsystem Object Reference=@n8, CameraLocation:Vector=@n10, CameraRotation:Rotator=@n9.CameraRotation) -> (then)
  ]}
 GetIfMapActive {nodes [
   GetIfMapActive() {is=title K2Node_FunctionEntry}
   n1 = Get CurrentBounds()
   n2 = Get CurrentRenderTarget() -> (CurrentRenderTarget:Texture Render Target 2D Object Reference)
   > Return Node!(
    execute
    isMapActive:bool=AND Boolean!(A=Get isMapActive() -> (isMapActive:bool), B=IsValid(Object=@n2), C:bool=Break Box 2D(Box2D=@n1) -> (bIsValid=~).bIsValid) -> (ReturnValue)
   )
  ]}
 OnMouseButtonDown {nodes [
   n1 = OnMouseButtonDown!() -> (then, MyGeometry:Geometry Structure, MouseEvent:"Pointer Event Structure (by ref)") {is=title K2Node_FunctionEntry}
   > n2 = Branch(Condition="Equal (Key)"(
     A:Key Structure=Get Effecting Button!(Input:"Pointer Event Structure (by ref)"=@n1.MouseEvent) -> (ReturnValue:Key Structure) {is=title K2Node_CallFunction}
     B:Key Structure=LeftMouseButton
    ) -> (ReturnValue:bool=~) {is=title K2Node_PromotableOperator})
   > n3 = GetMousePositionOnPlatform!(execute) -> (then, ReturnValue:Vector 2D Structure)
   > Set MouseDownLoc(MouseDownLoc=@n3)
   n4 = Get imgRT()
   > n5 = Branch(Condition=IsUnderLocation!(
     Geometry:"Geometry Structure (by ref)"=GetCachedGeometry(self=@n4)
     AbsoluteCoordinate:"Vector 2D Structure (by ref)"=Get MouseDownLoc() -> (MouseDownLoc:Vector 2D Structure)
    ) -> (ReturnValue:bool))
   CancelTeleportDraw(execute=@n5.else)
   CancelTeleportDraw(execute=@n2.else)
   Return Node!(execute, ReturnValue:Event Reply Structure)
  ]}
 OnMouseButtonUp {nodes [
   OnMouseButtonUp!() -> (then, MyGeometry:Geometry Structure, MouseEvent:"Pointer Event Structure (by ref)") {is=title K2Node_FunctionEntry}
   > n1 = Branch(Condition="Not Equal Exactly (Vector2D)"!(A:Vector 2D Structure=Get MouseDownLoc(), B:Vector 2D Structure=(X=-1,Y=-1)) -> (ReturnValue:bool))
   > n2 = Sequence()
   > n3 = GetMousePositionOnPlatform()
   n4 = To Direction And Length!(A:Vector 2D Structure=vector2d - vector2d(A=@n3, B=Get MouseDownLoc())) -> (OutDir:Vector 2D Structure, OutLength:double)
   > Set l_dragged(l_dragged:bool="float > float"(A=@n4.OutLength, B=15)) -> (Output_Get:bool)
   > n5 = GetCameraPositionFrom2DLoc!(execute, self, AbsoluteCoordinate:Vector 2D Structure=Get MouseDownLoc()) -> (then, location:Vector, rotation:Rotator) {is=title K2Node_CallFunction}
   n6 = BreakRotator(InRot=@n5.rotation)
   n7 = BreakVector2D(InVec=@n4)
   n8 = VectorToNormalizedAngleInDegrees!(V_X:double=@n7, V_Y:double=@n7.Y, V_Z:double=0) -> (ReturnValue)
   > SetLevelViewportCameraInfo(
    self=Get UnrealEditorSubsystem()
    CameraLocation=@n5
    CameraRotation=SelectRotator!(A:Rotator=MakeRotator(Pitch=@n6.Pitch, Yaw=@n8), B:Rotator=@n5.rotation, bPickA:bool=Get l_dragged() -> (l_dragged:bool)) -> (ReturnValue:Rotator)
   )
   CancelTeleportDraw(execute=@n2.then_1)
   Return Node(execute=@n2.then_2, ReturnValue=Handled())
   Return Node(execute=@n1.else, ReturnValue=Unhandled())
  ]}
 OnPaint {nodes [
   n1 = OnPaint!() -> (then, Context:"Paint Context Structure (by ref)") {is=title K2Node_FunctionEntry}
   > Set l_context(l_context:Paint Context Structure=@n1) -> (Output_Get:Paint Context Structure)
   > Sequence!(execute) -> (then_0, then_1)
   > Branch(Condition="Not Equal Exactly (Vector2D)"(A=Get MouseDownLoc(), B=(X=-1,Y=-1)))
   > n2 = GetMousePositionOnPlatform()
   > Set l_start(l_start:Vector 2D Structure="Absolute to Local (Coordinates)"(Geometry=GetCachedGeometry(), AbsoluteCoordinate=Get MouseDownLoc())) -> (Output_Get:Vector 2D Structure)
   > Set l_end(l_end:Vector 2D Structure="Absolute to Local (Coordinates)"(Geometry=GetCachedGeometry(), AbsoluteCoordinate=@n2)) -> (Output_Get:Vector 2D Structure)
   n3 = Get imgRT()
   n4 = "Absolute to Local (Coordinates)"(
    Geometry=GetCachedGeometry()
    AbsoluteCoordinate="Local to Absolute (Coordinates)"!(Geometry:"Geometry Structure (by ref)"=GetCachedGeometry(self=@n3), LocalCoordinate:Vector 2D Structure) -> (ReturnValue:Vector 2D Structure)
   )
   n5 = "To Vector (Vector2D)"!(InVector2D:Vector 2D Structure=@n4, Z:float=0) -> (ReturnValue:Vector)
   n6 = "To Vector (Vector2D)"(InVector2D=Get l_end() -> (l_end:Vector 2D Structure), Z=0)
   n7 = GetCachedGeometry(self=Get imgRT())
   n8 = "To Vector (Vector2D)"(
    InVector2D=vector2d + vector2d(
     A:Vector 2D Structure=@n4
     B:Vector 2D Structure=GetAbsoluteSize!(Geometry:"Geometry Structure (by ref)"=@n7) -> (ReturnValue:Vector 2D Structure)
    ) -> (ReturnValue:Vector 2D Structure) {is=title K2Node_PromotableOperator}
    Z=0
   )
   > Set l_end(l_end="To Vector2D (Vector)"!(InVector:Vector=Vector_BoundedToBox!(InVect:Vector=@n6, InBoxMin:Vector=@n5, InBoxMax:Vector=@n8) -> (ReturnValue:Vector)) -> (ReturnValue:Vector 2D Structure))
   > n9 = Sequence!(execute) -> (then_0, then_1, then_2:Exec)
   > DrawLine!(
    execute
    Context:"Paint Context Structure (by ref)"=Get l_context() -> (l_context:Paint Context Structure)
    PositionA:Vector 2D Structure=Get l_start() -> (l_start:Vector 2D Structure)
    PositionB:Vector 2D Structure=Get l_end()
    Tint:Linear Color Structure=(R=0.021,G=0.0412,B=0.099,A=1)
    bAntiAlias:bool=true
    Thickness:float=3
   ) -> (then)
   > DrawLine(Context=Get l_context(), PositionA=Get l_start(), PositionB=Get l_end(), Tint=(R=0.6354,G=0.2648,B=0.0975,A=1), bAntiAlias=true, Thickness=2)
   n10 = MakeVector2D(X=12, Y=12)
   DrawBox!(
    execute=@n9.then_1
    Context:"Paint Context Structure (by ref)"=Get l_context()
    Position:Vector 2D Structure=vector2d - vector2d(A=Get l_start(), B=vector2d / vector2d(A=@n10, B:float=2))
    Size:Vector 2D Structure=@n10
    Brush:Slate Brush Asset Object Reference=/Game/Blueprints/Blutilities/ForCinematics/PlaceByUVTool/drawTestBrush
    Tint:Linear Color Structure=(R=0.7917,G=0.3315,B=0.0735,A=1)
   ) -> (then)
   n11 = MakeLiteralText(Value="INVTEXT(\"Right Click To Cancel\")")
   n12 = vector2d + vector2d(A=Get l_end(), B=(X=-12,Y=15))
   n13 = Get l_context()
   Draw Text!(
    execute=@n9.then_2
    Context:"Paint Context Structure (by ref)"=@n13
    Text:"Text (by ref)"=@n11
    Position:Vector 2D Structure=vector2d + vector2d(A=@n12, B=(X=-1,Y=-1))
    Font:Font Object Reference=/LauncherSocial/BrutalType
    FontSize:float=12
    FontTypeFace:Name=Bold
    Tint:Linear Color Structure=(R=0.0068,G=0.0122,B=0.0365,A=1)
   ) -> (then)
   > Draw Text(Context=@n13, Text=@n11, Position=@n12, Font=&BrutalType, FontSize=12, FontTypeFace=Bold, Tint=(R=0.5,G=0.353,B=0.1963,A=1))
   Return Node!(execute)
  ]}
 SetOverlayTexture {nodes [
   n1 = SetOverlayTexture!() -> (then, OverlayTexture:Texture Object Reference) {is=title K2Node_FunctionEntry}
   > Set OverlayTexture(OverlayTexture:Texture Object Reference=@n1) -> (Output_Get:Texture Object Reference)
   > ValidateMapMaterial()
  ]}
 TryGetBounds {nodes [
   TryGetBounds() {is=title K2Node_FunctionEntry}
   > ClearMessage(self:EUW Top Down Map Errors Object Reference=Get EUW_TopDownMapErrors() -> (EUW_TopDownMapErrors:EUW Top Down Map Errors Object Reference))
   > n1 = Sequence()
   > n2 = Sequence!(execute) -> (then_0, then_1)
   > n3 = GetBoundsFromActorByTag!(execute, self) -> (then, ReturnValue:Box 2D Structure) {is=title K2Node_CallFunction}
   > Set CurrentBounds(CurrentBounds:Box 2D Structure=@n3) -> (Output_Get:Box 2D Structure)
   n4 = Branch(execute=@n2.then_1, Condition=Break Box 2D(Box2D=Get CurrentBounds()) -> (bIsValid=~).bIsValid)
   > Return Node!(execute, CurrentBounds:Box 2D Structure=Get CurrentBounds())
   AddMessage!(
    execute=@n4.else
    self:EUW Top Down Map Errors Object Reference=Get EUW_TopDownMapErrors()
    message:String=Find(TargetMap=Get ErrorMap() -> (ErrorMap:Map of Strings to Strings), Key=boundsActor)
   ) -> (then)
   n5 = BreakVector2D(InVec=Get BackupBoundsMax() -> (BackupBoundsMax:Vector 2D Structure))
   n6 = BreakVector2D(InVec=Get BackupBoundsMin() -> (BackupBoundsMin:Vector 2D Structure))
   n7 = "float > float"(A=@n5.Y, B=@n6.Y)
   n8 = "float > float"(A=@n5, B=@n6)
   n9 = Get BackupBoundsMin()
   n10 = Get BackupBoundsMax()
   n11 = Branch(
    execute=@n1.then_1
    Condition=AND Boolean!(
     A="Not Equal (Vector2D)"!(A:Vector 2D Structure=@n9, B:Vector 2D Structure=@n10, ErrorTolerance:float) -> (ReturnValue:bool=~) {is=title K2Node_PromotableOperator}
     B=AND Boolean(A=@n8, B=@n7)
    ) -> (ReturnValue)
   )
   > Set CurrentBounds(CurrentBounds=MakeBox2D(Min=Get BackupBoundsMin(), Max=Get BackupBoundsMax()))
   > Return Node(CurrentBounds=Get CurrentBounds())
   AddMessage(execute=@n11.else, self=Get EUW_TopDownMapErrors(), message=Find(TargetMap=Get ErrorMap(), Key=boundsStatic))
   Return Node(execute=@n1.then_2, CurrentBounds=Get CurrentBounds())
  ]}
 ValidateMapMaterial {nodes [
   ValidateMapMaterial() {is=title K2Node_FunctionEntry}
   > n1 = Sequence!(execute) -> (then_0, then_1, then_2:Exec)
   > n2 = Cast To MaterialInstanceDynamic!(
    execute
    Object=Get!(self:Image Object Reference=Get imgRT()) -> (
     Brush_TintColor:Slate Color Structure
     Brush_DrawAs:ESlateBrushDrawType Enum
     Brush_Tiling:ESlateBrushTileType Enum
     Brush_Mirroring:ESlateBrushMirrorType Enum
     Brush_ImageSize:Vector 2D Structure
     Brush_Margin:Margin Structure
     Brush_ResourceObject:Object Reference
     Brush_OutlineSettings:Slate Brush Outline Settings Structure
    ).Brush_ResourceObject
   ) -> (then, CastFailed, AsMaterial Instance Dynamic:Material Instance Dynamic Object Reference) {is=title K2Node_DynamicCast}
   > n3 = Set l_MID(l_MID:Material Instance Dynamic Object Reference=@n2) -> (Output_Get:Material Instance Dynamic Object Reference)
   SetBrushFromMaterial(execute=[@n3.then, @n5.then], self=Get imgRT(), Material=Get l_MID() -> (l_MID:Material Instance Dynamic Object Reference))
   n4 = CreateDynamicMaterialInstance(
    execute=@n2.CastFailed
    WorldContextObject=Self-Reference()
    Parent=/Game/NeverCook/EditorUtilities/EditorTopDownMap/Classes/M_TopDownMapDisplay
    OptionalName=None
    CreationFlags=Transient
   )
   > n5 = Set l_MID(l_MID=@n4)
   Is Valid(exec=@n1.then_1, InputObject=Get CurrentRT())
   > SetTextureParameterValue(self=Get l_MID(), ParameterName=MapTexture, Value=Get CurrentRT())
   n6 = Is Valid(exec=@n1.then_2, InputObject=Get OverlayTexture() -> (OverlayTexture:Texture Object Reference))
   SetScalarParameterValue(execute=@n6.Is Not Valid, self=Get l_MID(), ParameterName=BlendOverlay, Value=0)
   SetTextureParameterValue(execute=@n6.Is Valid, self=Get l_MID(), ParameterName=OverlayTexture, Value=Get OverlayTexture())
   > SetScalarParameterValue(self=Get l_MID(), ParameterName=BlendOverlay, Value=Get OverlayBlend() -> (OverlayBlend:float))
  ]}
]
variables [
 ActorTagForBounds {type=string, default=HeatmapBoundsActor}
 BackupBoundsMax {type=Vector2D, default=<15000 15000>}
 BackupBoundsMin {type=Vector2D, default=<-15000 -15000>}
 CaptureActorBP {type=Object, default=/Game/NeverCook/EditorUtilities/EditorTopDownMap/Classes/TopDownMap_CaptureActor}
 CurrentBounds {type=Box2D, default {bIsValid=false, max=<0 0>, min=<0 0>}}
 CurrentRenderTarget {type=TextureRenderTarget2D, default=/Game/NeverCook/EditorUtilities/TextureHeatmapRenderer/Art/TextureHeatmapRenderer_Target2D}
 CurrentRT {type=TextureRenderTarget2D, default=null}
 ErrorMap {
  type="map<string>"
  default {
   boundsActor=Could not find an actor using the tag specified in ActorTagForBounds.
   boundsStatic="No valid Min/Max Bounds is set.  Max bounds must be larger than min Bounds"
  }
 }
 isMapActive {type=boolean, default=true}
 lastCamDirection {type=Vector}
 lastCamLocation {type=Vector}
 MouseDownLoc {type=Vector2D, default=<-1 -1>}
 OverlayBlend {type=number, default=0.5}
 OverlayTexture {type=Texture, default=&TextureHeatmapRenderer_Target2D}
 useActorBounds {type=boolean, default=true}
]
widgets {
 info {rootWidgetClass=&StackBox, widgetCount=10}
 widgets [
  StackBox_0 {widgetClassPath=&StackBox, widget=StackBox_0, properties {orientation=Orient_Vertical}}
  HorizontalBox_0 {
   widgetClassPath=&HorizontalBox
   parent=StackBox_0
   widget=HorizontalBox_0
   slot {class=StackBoxSlot, properties {size {sizeRule=Fill, value=1}}}
  }
  Border_0 {
   widgetClassPath=&Border
   parent=HorizontalBox_0
   widget=Border_0
   properties {brushColor=<rgba 0.1406 0.1406 0.1406 1>}
   slot {class=HorizontalBoxSlot, properties {size {sizeRule=Fill, value=2}}}
  }
  Overlay_1 {widgetClassPath=&Overlay, parent=Border_0, widget=Overlay_1, slot {class=BorderSlot}}
  ScaleBox_0 {
   widgetClassPath=&ScaleBox
   parent=Overlay_1
   widget=ScaleBox_0
   properties {stretchDirection=DownOnly}
   slot {class=OverlaySlot, properties {horizontalAlignment=HAlign_Fill, verticalAlignment=VAlign_Fill}}
  }
  Overlay_0 {widgetClassPath=&Overlay, parent=ScaleBox_0, widget=Overlay_0, slot {class=ScaleBoxSlot}}
  imgRT {
   widgetClassPath=&Image
   bIsVariable=true
   parent=Overlay_0
   widget=imgRT
   properties {brush {
     bIsDynamicallyLoaded=false
     drawAs=Image
     imageType=NoImage
     mirroring=NoMirror
     resourceName=None
     resourceObject=&M_TopDownMapDisplay
     tiling=NoTile
     imageSize=<512 512>
     margin=<ltrb 0 0 0 0>
     outlineSettings {
      bUseBrushTransparency=false
      roundingType=HalfHeightRadius
      width=0
      color {colorUseRule=UseColor_Specified, specifiedColor=<rgba 0 0 0 0>}
      cornerRadii=<0 0 0 0>
     }
     tintColor {colorUseRule=UseColor_Specified, specifiedColor=<rgba 1 1 1 1>}
     uVRegion {bIsValid=false, max=<0 0>, min=<0 0>}
    }}
   slot {class=OverlaySlot}
  }
  imgCam {
   widgetClassPath=&Image
   bIsVariable=true
   parent=Overlay_0
   widget=imgCam
   properties {
    brush {
     bIsDynamicallyLoaded=false
     drawAs=Image
     imageType=NoImage
     mirroring=NoMirror
     resourceName=None
     resourceObject=&Icon_CameraWithFrustum
     tiling=NoTile
     imageSize=<32 32>
     margin=<ltrb 0 0 0 0>
     outlineSettings {
      bUseBrushTransparency=false
      roundingType=HalfHeightRadius
      width=0
      color {colorUseRule=UseColor_Specified, specifiedColor=<rgba 0 0 0 0>}
      cornerRadii=<0 0 0 0>
     }
     tintColor {colorUseRule=UseColor_Specified, specifiedColor=<rgba 1 1 1 1>}
     uVRegion {bIsValid=false, max=<0 0>, min=<0 0>}
    }
    renderTransform {angle=0, scale=<1 1>, shear=<0 0>, translation=<0 -12>}
   }
   slot {class=OverlaySlot}
  }
  EUW_TopDownMapOptions {
   widgetClassPath=&/EUW_TopDownMapOptions
   bIsVariable=true
   parent=Overlay_1
   widget=EUW_TopDownMapOptions
   slot {class=OverlaySlot, properties {horizontalAlignment=HAlign_Fill, verticalAlignment=VAlign_Fill}}
  }
  EUW_TopDownMapErrors {
   widgetClassPath=&/EUW_TopDownMapErrors
   bIsVariable=true
   parent=Overlay_1
   widget=EUW_TopDownMapErrors
   slot {class=OverlaySlot, properties {horizontalAlignment=HAlign_Fill, verticalAlignment=VAlign_Fill}}
  }
 ]
}
