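# Editor utility widget that displays error and warning messages, apparently for the top-down map editor tool.
# AddMessage appends non-empty strings as new lines and ClearMessage, also run on construct, empties them; both call ValdateMessage.
# ValdateMessage refreshes the scrollable text and collapses the whole widget when the string is empty.
# A toggle button shows or hides the initially collapsed message panel, and btnX collapses it.
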EditorUtilityWidgetBlueprint /Game/NeverCook/EditorUtilities/EditorTopDownMap/EUW_TopDownMapErrors
graphs [
 AddMessage {nodes [
   n1 = AddMessage!() -> (then, message:String) {is=title K2Node_FunctionEntry}
   > Branch(Condition="Not Equal Exactly (String)"(A=@n1))
   > Set currentMessageString(currentMessageString:String=Append!(A=Get currentMessageString() -> (currentMessageString:String), B="\r\n", C:String=@n1) -> (ReturnValue)) -> (Output_Get:String)
   > ValdateMessage()
  ]}
 ClearMessage {nodes [
   ClearMessage() {is=title K2Node_FunctionEntry}
   > Set currentMessageString()
   > ValdateMessage()
  ]}
 EventGraph {nodes [
   n1 = FindParentWidgetOfType!(
    execute
    StartingWidget:Widget Object Reference=Self-Reference()
    Type:Widget Class Reference=&/EUW_TopDownMapWidget.EUW_TopDownMapWidget_C
   ) -> (then, ReturnValue:EUW Top Down Map Widget Object Reference)
   > Is Valid(InputObject=@n1)
   > Set ParentWidget(ParentWidget:EUW Top Down Map Widget Object Reference=@n1) -> (Output_Get:EUW Top Down Map Widget Object Reference)
   "On Clicked (btnX)"!() -> (OutputDelegate, then, widget:User Widget Object Reference, event:Pointer Event Structure) {is=title K2Node_ComponentBoundEvent}
   > SetVisibility(self=Get Border_0() -> (Border_0:Border Object Reference), InVisibility=Collapsed)
   "On Clicked (btnToggle)"() {is=title K2Node_ComponentBoundEvent}
   n2 = Get Border_0()
   > SetVisibility(self=@n2, InVisibility=Select(Option 0=SelfHitTestInvisible, Option 1=Collapsed, Index=IsVisible(self=@n2)))
   Event Construct()
   > ClearMessage() {is=title K2Node_CallFunction}
  ]}
 ValdateMessage {nodes [
   ValdateMessage() {is=title K2Node_FunctionEntry}
   > "SetText (Multi-Line Editable Text)"!(
    execute
    self:"Editable Text (Multi- Line) Object Reference"=Get EditorUtilityMultiLineEditableText_175() -> (EditorUtilityMultiLineEditableText_175:Editor Utility Multi Line Editable Text Object Reference)
    InText:Text="To Text (String)"(InString=Get currentMessageString())
   ) -> (then)
   n1 = Get currentMessageString()
   > SetVisibility(InVisibility=Select(Option 0=SelfHitTestInvisible, Option 1=Collapsed, Index=IsEmpty(InString=@n1)))
  ]}
]
variables [
 currentMessageString {type=string, default=""}
 ParentWidget {type=EUW_TopDownMapWidget_C, default=null}
]
widgets {
 info {rootWidgetClass=&Overlay, widgetCount=8}
 widgets [
  Overlay_0 {widgetClassPath=&Overlay, widget=Overlay_0}
  btnToggle {
   widgetClassPath=&EditorUtilityButton
   parent=Overlay_0
   widget=btnToggle
   properties {widgetStyle {
     clickedSlateSound {resourceObject=null}
     disabled {
      bIsDynamicallyLoaded=false
      drawAs=Image
      imageType=NoImage
      mirroring=NoMirror
      resourceName=None
      resourceObject=/Game/Athena/UI/InGame/Creative/PropertyEditor/Icon_Picker/Textures_V2/64x64/T_UI_IconLibrary_Gear_64
      tiling=NoTile
      imageSize=<32 32>
      margin=<ltrb 0 0 0 0>
      outlineSettings {
       bUseBrushTransparency=false
       roundingType=FixedRadius
       width=1
       color {colorUseRule=UseColor_ColorTable, specifiedColor=<rgba 0 0 0 0>}
       cornerRadii=<4 4 4 4>
      }
      tintColor {colorUseRule=UseColor_Specified, specifiedColor=<rgba 0.1823 0.1823 0.1823 1>}
      uVRegion {bIsValid=false, max=<0 0>, min=<0 0>}
     }
     disabledForeground {colorUseRule=UseColor_ColorTable, specifiedColor=<rgba 0 0 0 0>}
     hovered {
      bIsDynamicallyLoaded=false
      drawAs=Image
      imageType=NoImage
      mirroring=NoMirror
      resourceName=None
      resourceObject=null
      tiling=NoTile
      imageSize=<32 32>
      margin=<ltrb 0 0 0 0>
      outlineSettings {
       bUseBrushTransparency=false
       roundingType=FixedRadius
       width=1
       color {colorUseRule=UseColor_ColorTable, specifiedColor=<rgba 0 0 0 0>}
       cornerRadii=<4 4 4 4>
      }
      tintColor {colorUseRule=UseColor_Specified, specifiedColor=<rgba 1 1 1 1>}
      uVRegion {bIsValid=false, max=<0 0>, min=<0 0>}
     }
     hoveredForeground {colorUseRule=UseColor_ColorTable, specifiedColor=<rgba 0 0 0 0>}
     hoveredSlateSound {resourceObject=null}
     normal {
      bIsDynamicallyLoaded=false
      drawAs=Image
      imageType=NoImage
      mirroring=NoMirror
      resourceName=None
      resourceObject=null
      tiling=NoTile
      imageSize=<32 32>
      margin=<ltrb 0 0 0 0>
      outlineSettings {
       bUseBrushTransparency=false
       roundingType=FixedRadius
       width=1
       color {colorUseRule=UseColor_ColorTable, specifiedColor=<rgba 0 0 0 0>}
       cornerRadii=<4 4 4 4>
      }
      tintColor {colorUseRule=UseColor_Specified, specifiedColor=<rgba 1 1 1 0.664>}
      uVRegion {bIsValid=false, max=<0 0>, min=<0 0>}
     }
     normalForeground {colorUseRule=UseColor_ColorTable, specifiedColor=<rgba 0 0 0 0>}
     normalPadding=<ltrb 12 1.5 12 1.5>
     pressed {
      bIsDynamicallyLoaded=false
      drawAs=Image
      imageType=NoImage
      mirroring=NoMirror
      resourceName=None
      resourceObject=null
      tiling=NoTile
      imageSize=<32 32>
      margin=<ltrb 0 0 0 0>
      outlineSettings {
       bUseBrushTransparency=false
       roundingType=FixedRadius
       width=1
       color {colorUseRule=UseColor_ColorTable, specifiedColor=<rgba 0 0 0 0>}
       cornerRadii=<4 4 4 4>
      }
      tintColor {colorUseRule=UseColor_Specified, specifiedColor=<rgba 0.4688 0.4688 0.4688 0.664>}
      uVRegion {bIsValid=false, max=<0 0>, min=<0 0>}
     }
     pressedForeground {colorUseRule=UseColor_ColorTable, specifiedColor=<rgba 0 0 0 0>}
     pressedPadding=<ltrb 12 2.5 12 0.5>
     pressedSlateSound {resourceObject=null}
    }}
   slot {class=OverlaySlot, properties {verticalAlignment=VAlign_Bottom}}
  }
  Border_0 {
   widgetClassPath=&Border
   bIsVariable=true
   parent=Overlay_0
   widget=Border_0
   properties {visibility=Collapsed, brushColor=<rgba 0.5573 0.088 0.038 1>, padding=<ltrb 2 2 2 0>}
   slot {class=OverlaySlot, properties {horizontalAlignment=HAlign_Fill, verticalAlignment=VAlign_Bottom}}
  }
  Border_1 {
   widgetClassPath=&Border
   parent=Border_0
   widget=Border_1
   properties {brushColor=<rgba 0.0417 0.0372 0.0347 1>}
   slot {class=BorderSlot, properties {padding=<ltrb 2 2 2 0>}}
  }
  VerticalBox_0 {widgetClassPath=&VerticalBox, parent=Border_1, widget=VerticalBox_0, slot {class=BorderSlot}}
  btnX {
   widgetClassPath=&TA_Generic01_Button
   parent=VerticalBox_0
   widget=btnX
   properties {
    buttonColor_Default=<rgba 0.6667 0.6667 0.6667 1>
    iconBrush {
     bIsDynamicallyLoaded=false
     drawAs=Image
     imageType=NoImage
     mirroring=NoMirror
     resourceName=None
     resourceObject=/Game/UI/CodeReferenced/UI_QuickbarDownArrow
     tiling=NoTile
     imageSize=<12 12>
     margin=<ltrb 0 0 0 0>
     outlineSettings {
      bUseBrushTransparency=false
      roundingType=HalfHeightRadius
      width=0
      color {colorUseRule=UseColor_Specified, specifiedColor=<rgba 0 0 0 0>}
      cornerRadii=<0 0 0 1>
     }
     tintColor {colorUseRule=UseColor_Specified, specifiedColor=<rgba 0.0417 0.0417 0.0417 1>}
     uVRegion {bIsValid=false, max=<0 0>, min=<0 0>}
    }
   }
   slot {class=VerticalBoxSlot, properties {horizontalAlignment=HAlign_Left, padding=<ltrb 0 0 0 5>}}
  }
  EditorUtilityScrollBox_0 {
   widgetClassPath=&EditorUtilityScrollBox
   parent=VerticalBox_0
   widget=EditorUtilityScrollBox_0
   slot {class=VerticalBoxSlot}
  }
  EditorUtilityMultiLineEditableText_175 {
   widgetClassPath=/Script/Blutility.EditorUtilityMultiLineEditableText
   bIsVariable=true
   parent=EditorUtilityScrollBox_0
   widget=EditorUtilityMultiLineEditableText_175
   properties {
    text="Error !\r\nWarning !"
    widgetStyle {
     overflowPolicy=Clip
     transformPolicy=None
     colorAndOpacity {colorUseRule=UseColor_Foreground, specifiedColor=<rgba 1 0 1 1>}
     font {
      bForceMonospaced=false
      bMaterialIsStencil=false
      fontMaterial=null
      fontObject=null
      letterSpacing=0
      monospacedWidth=1
      size=9
      skewAmount=0
      typefaceFontName=Regular
      outlineSettings {
       bApplyOutlineToDropShadows=false
       bMiteredCorners=false
       bSeparateFillAlpha=false
       outlineMaterial=null
       outlineSize=0
       outlineColor=<rgba 0 0 0 1>
      }
     }
     highlightColor {colorUseRule=UseColor_ColorTable, specifiedColor=<rgba 0 0 0 0>}
     highlightShape {
      bIsDynamicallyLoaded=false
      drawAs=Image
      imageType=NoImage
      mirroring=NoMirror
      resourceName=None
      resourceObject=null
      tiling=NoTile
      imageSize=<0 0>
      margin=<ltrb 0 0 0 0>
      outlineSettings {
       bUseBrushTransparency=false
       roundingType=HalfHeightRadius
       width=0
       color {colorUseRule=UseColor_Specified, specifiedColor=<rgba 0 0 0 0>}
       cornerRadii=<0 0 0 0>
      }
      tintColor {colorUseRule=UseColor_Specified, specifiedColor=<rgba 0.2582 0.5395 0.0685 1>}
      uVRegion {bIsValid=false, max=<0 0>, min=<0 0>}
     }
     selectedBackgroundColor {colorUseRule=UseColor_Specified, specifiedColor=<rgba 0 0.162 0.7454 1>}
     shadowColorAndOpacity=<rgba 0 0 0 1>
     shadowOffset=<0 0>
     strikeBrush {
      bIsDynamicallyLoaded=false
      drawAs=Image
      imageType=NoImage
      mirroring=NoMirror
      resourceName=None
      resourceObject=null
      tiling=NoTile
      imageSize=<32 32>
      margin=<ltrb 0 0 0 0>
      outlineSettings {
       bUseBrushTransparency=false
       roundingType=HalfHeightRadius
       width=0
       color {colorUseRule=UseColor_Specified, specifiedColor=<rgba 0 0 0 0>}
       cornerRadii=<0 0 0 0>
      }
      tintColor {colorUseRule=UseColor_Specified, specifiedColor=<rgba 1 1 1 1>}
      uVRegion {bIsValid=false, max=<0 0>, min=<0 0>}
     }
     underlineBrush {
      bIsDynamicallyLoaded=false
      drawAs=Image
      imageType=NoImage
      mirroring=NoMirror
      resourceName=None
      resourceObject=null
      tiling=NoTile
      imageSize=<32 32>
      margin=<ltrb 0 0 0 0>
      outlineSettings {
       bUseBrushTransparency=false
       roundingType=HalfHeightRadius
       width=0
       color {colorUseRule=UseColor_Specified, specifiedColor=<rgba 0 0 0 0>}
       cornerRadii=<0 0 0 0>
      }
      tintColor {colorUseRule=UseColor_Specified, specifiedColor=<rgba 1 1 1 1>}
      uVRegion {bIsValid=false, max=<0 0>, min=<0 0>}
     }
    }
   }
   slot {class=ScrollBoxSlot}
  }
 ]
}
